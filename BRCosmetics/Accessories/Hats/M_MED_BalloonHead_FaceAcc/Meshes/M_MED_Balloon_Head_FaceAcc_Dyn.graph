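# Physics asset for the balloon head face accessory in BRCosmetics, simulating a balloon on a string attached to the character's upper spine.
# A constraint chain links the balloon capsule through two thin string capsules to spine_05, with wide swing and tight twist limits.
# A large spine_04 capsule and five small angled capsules on spine_05 form the torso shapes, joined by an angularly free constraint.
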PhysicsAsset /BRCosmetics/Accessories/Hats/M_MED_BalloonHead_FaceAcc/Meshes/M_MED_Balloon_Head_FaceAcc_Dyn
bodies {
 dyn_balloon [{shapeType=Capsule, length=4.9583, radius=10.1059, center=<14.1638 0.5193 0>, rotation=<rot 90 0 0>}]
 dyn_string_01 [{shapeType=Capsule, length=2.4917, radius=0.577, center=<1.8112 0 0>, rotation=<rot 90 0 0>}]
 dyn_string_02 [{shapeType=Capsule, length=1.5094, radius=0.5075, center=<1.6835 0 0>, rotation=<rot 90 0 0>}]
 spine_04 [{shapeType=Capsule, length=7.1723, radius=19.1497, center=<-2.8116 1.376 0>, rotation=<rot 90 0 0>}]
 spine_05 [
  {shapeType=Capsule, length=8.1946, radius=1.5116, center=<20.3196 -6.2593 -0.0001>}
  {shapeType=Capsule, length=8.1946, radius=1.5116, center=<18.6317 -3.4442 -8.1316>, rotation=<rot -17.3871 18.2503 -37.8114>}
  {shapeType=Capsule, length=8.1946, radius=1.5116, shapeName=Side, center=<14.2517 3.1251 11.7458>, rotation=<rot 2.5103 23.6433 89.408>}
  {shapeType=Capsule, length=8.1946, radius=1.5116, center=<18.6317 -3.4442 8.1316>, rotation=<rot 17.3871 18.2503 37.8114>}
  {shapeType=Capsule, length=8.1946, radius=1.5116, shapeName=Side, center=<14.2488 3.123 -11.8251>, rotation=<rot -2.5103 23.6434 -89.408>}
 ]
}
constraints [
 {bone1Name=dyn_balloon, bone2Name=dyn_string_02, swing1LimitDegrees=90, swing2LimitDegrees=90, twistLimitDegrees=10}
 {bone1Name=dyn_string_02, bone2Name=dyn_string_01, swing1LimitDegrees=90, swing2LimitDegrees=90, twistLimitDegrees=10}
 {bone1Name=dyn_string_01, bone2Name=spine_05, swing1LimitDegrees=90, swing2LimitDegrees=90, twistLimitDegrees=10}
 {
  bone1Name=spine_05
  bone2Name=spine_04
  swing1LimitDegrees=45
  swing1Motion=Free
  swing2LimitDegrees=45
  swing2Motion=Free
  twistLimitDegrees=45
  twistMotion=Free
 }
]
