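# Animation Blueprint for the Balloon Head face accessory, animating a floating balloon and its string on the character.
# It copies the attached parent's mesh pose, then runs RigidBody simulation in spine_05 space with upward gravity, blended by an interpolated PawnSpeedAlpha.
# Two constraints then place the string bones between spine_05 and the simulated balloon using opposite weightings, so the string follows without its own simulation result.
# The EventGraph is empty.
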AnimBlueprint /BRCosmetics/Accessories/Hats/M_MED_BalloonHead_FaceAcc/Meshes/M_MED_Balloon_Head_FaceAcc_AnimBP
target_skeleton=&/M_MED_Balloon_Head_FaceAcc_Skeleton
graphs [
 AnimGraph {nodes [
   n1 = RigidBody(
    ComponentPose=Local To Component(LocalPose=Copy Pose From Mesh() {bUseAttachedParent=true})
    Alpha=Get PawnSpeedAlpha() -> (PawnSpeedAlpha:float)
   ) {
    bClampLinearTranslationLimitToRefPose=true
    bOverrideWorldGravity=true
    overridePhysicsAsset=&/M_MED_Balloon_Head_FaceAcc_Dyn
    simulationSpace=BaseBoneSpace
    alphaBoolBlend {blendOption=Linear}
    alphaScaleBiasClamp {bInterpResult=true, interpSpeedDecreasing=2, interpSpeedIncreasing=2}
    baseBoneRef {boneName=spine_05}
    componentAppliedLinearAccClamp=<3000 3000 3000>
    componentLinearAccScale=<0.27 0.2 0.1>
    componentLinearVelScale=<0.27 0.2 0.1>
    overrideWorldGravity {z=300}
   }
   n2 = "Constraint - dyn_string_01 (PP)"!(ComponentPose=@n1, Alpha=1, ConstraintWeights_0=0.75, ConstraintWeights_1:float=0.25) -> (Pose) {
    alphaBoolBlend {blendOption=Linear}
    boneToModify {boneName=dyn_string_01}
    constraintSetup [{targetBone {boneName=spine_05}}, {targetBone {boneName=dyn_balloon}}]
    constraintWeights [1,1]
   }
   n3 = "Constraint - dyn_string_02 (PP)"!(ComponentPose=@n2, Alpha=1, ConstraintWeights_0=0.25, ConstraintWeights_1:float=0.75) -> (Pose) {
    alphaBoolBlend {blendOption=Linear}
    boneToModify {boneName=dyn_string_02}
    constraintSetup [{targetBone {boneName=spine_05}}, {targetBone {boneName=dyn_balloon}}]
    constraintWeights [1,1]
   }
   Output Pose(Result=Component To Local(ComponentPose=@n3)) {layerGroup=None}
  ]}
 EventGraph {links [], nodes []}
]
variables []
